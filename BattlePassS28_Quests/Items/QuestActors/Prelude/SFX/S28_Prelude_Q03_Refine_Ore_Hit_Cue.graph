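# Sound cue for the ore-refining hit in Season 28 Battle Pass prelude quest Q03.
# A random node picks one of six hit waves with equal weight, then a modulator adds slight pitch and volume variation.
# It plays in the quests sound class and borrows the player pickaxe impact attenuation for spatial falloff.
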SoundCue /BattlePassS28_Quests/Items/QuestActors/Prelude/SFX/S28_Prelude_Q03_Refine_Ore_Hit_Cue
root=Modulator_0
nodes [
 {id=Modulator_0, children [Random_0], properties {pitchMax=1.02, pitchMin=0.98, volumeMax=1.02, volumeMin=0.98}}
 {id=Random_0, children [WavePlayer_0,WavePlayer_1,WavePlayer_2,WavePlayer_3,WavePlayer_4,WavePlayer_5], properties {weights [1,1,1,1,1,1]}}
 {id=WavePlayer_0, properties {soundWaveAssetPtr=&/S28_Prelude_Refine_Ore_Hit_01}}
 {id=WavePlayer_1, properties {soundWaveAssetPtr=&/S28_Prelude_Refine_Ore_Hit_02}}
 {id=WavePlayer_2, properties {soundWaveAssetPtr=&/S28_Prelude_Refine_Ore_Hit_03}}
 {id=WavePlayer_3, properties {soundWaveAssetPtr=&/S28_Prelude_Refine_Ore_Hit_04}}
 {id=WavePlayer_4, properties {soundWaveAssetPtr=&/S28_Prelude_Refine_Ore_Hit_05}}
 {id=WavePlayer_5, properties {soundWaveAssetPtr=&/S28_Prelude_Refine_Ore_Hit_06}}
]
properties {
 attenuationSettings=/Game/Sounds/Attenuation/Fort_Weapon_Pickaxe_Player_Impact_Att
 maxDistance=3512
 soundClassObject=&Fort_Quests
}
